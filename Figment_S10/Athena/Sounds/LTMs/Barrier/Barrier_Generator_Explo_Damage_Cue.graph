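# Sound cue for the explosion damage sound of the barrier generator in the Barrier limited-time mode.
# It picks one of three reused airstrike hit waves at equal odds, then passes it through a modulator for random pitch variation, with an overall pitch and volume boost on the cue.
# It uses large close/distant event attenuation and a weapon projectile sound class.
# A concurrency override caps simultaneous instances, stopping the oldest.
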SoundCue /Figment_S10/Athena/Sounds/LTMs/Barrier/Barrier_Generator_Explo_Damage_Cue
root=Modulator_0
nodes [
 {id=Modulator_0, children [Random_1], properties {pitchMax=1.3}}
 {id=Random_1, children [WavePlayer_0,WavePlayer_1,WavePlayer_2], properties {weights [1,1,1]}}
 {id=WavePlayer_0, properties {soundWaveAssetPtr=/Game/Sounds/Fort_Gadgets/Airstrike/Airstrik_Hit_01}}
 {id=WavePlayer_1, properties {soundWaveAssetPtr=/Game/Sounds/Fort_Gadgets/Airstrike/Airstrik_Hit_02}}
 {id=WavePlayer_2, properties {soundWaveAssetPtr=/Game/Sounds/Fort_Gadgets/Airstrike/Airstrik_Hit_03}}
]
properties {
 attenuationSettings=/Game/Athena/Sounds/Events/Attenuations/Events_Large_CloseDistant_Att
 bOverrideConcurrency=true
 maxDistance=30000
 pitchMultiplier=1.25
 soundClassObject=/Figment_S10/Sounds/FortSoundClassesAndModes/Fort_Weapon_Projectile_PSM
 volumeMultiplier=1.5
 concurrencyOverrides {maxCount=3, resolutionRule=StopOldest}
}
