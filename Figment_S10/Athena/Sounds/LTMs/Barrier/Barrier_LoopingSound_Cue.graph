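# Looping Sound Cue for the LTM Barrier, layering a moving-wall loop with close and distant moving-debris loops.
# The wall loop's pitch and volume follow MovingParam2, while the mixed debris pair takes pitch from MovingParam3 and volume from MovingParam2Inv.
# Each of the three layers has its own attenuation and a quality-level switch between the regular and "_opt" waves.
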SoundCue /Figment_S10/Athena/Sounds/LTMs/Barrier/Barrier_LoopingSound_Cue
root=Mixer_0
nodes [
 {
  id=Attenuation_0
  children [ModulatorContinuous_0]
  properties {attenuationSettings=&/Barrier_Wall_Idle_Att}
 }
 {id=Attenuation_2, children [QualityLevel_1], properties {attenuationSettings=&/Barrier_Wall_Close_Att}}
 {
  id=Attenuation_3
  children [QualityLevel_2]
  properties {attenuationSettings=&/Barrier_Wall_Distant_Att}
 }
 {id=Mixer_0, children [Attenuation_0, ModulatorContinuous_2], properties {inputVolume [1,1]}}
 {id=Mixer_2, children [Attenuation_2, Attenuation_3], properties {inputVolume [1,1]}}
 {
  id=ModulatorContinuous_0
  children [QualityLevel_0]
  properties {pitchModulationParams {minOutput=0.7, parameterName=MovingParam2}, volumeModulationParams {parameterName=MovingParam2}}
 }
 {
  id=ModulatorContinuous_2
  children [Mixer_2]
  properties {
   pitchModulationParams {default=0, maxOutput=1.2, minOutput=0.6, parameterName=MovingParam3}
   volumeModulationParams {default=0, parameterName=MovingParam2Inv}
  }
 }
 {id=QualityLevel_0, children [WavePlayer_1,WavePlayer_1,WavePlayer_4,WavePlayer_4]}
 {id=QualityLevel_1, children [WavePlayer_3,WavePlayer_3,WavePlayer_5,WavePlayer_5]}
 {id=QualityLevel_2, children [WavePlayer_2,WavePlayer_2,WavePlayer_6,WavePlayer_6]}
 {id=WavePlayer_1, properties {bLooping=true, soundWaveAssetPtr=&/Barrier_MovingLoop_01}}
 {id=WavePlayer_2, properties {bLooping=true, soundWaveAssetPtr=&/Barrier_MovingDebrisFar_01}}
 {id=WavePlayer_3, properties {bLooping=true, soundWaveAssetPtr=&/Barrier_MovingDebrisClose_01}}
 {id=WavePlayer_4, properties {bLooping=true, soundWaveAssetPtr=&/Barrier_MovingLoop_01_opt}}
 {id=WavePlayer_5, properties {bLooping=true, soundWaveAssetPtr=&/Barrier_MovingDebrisClose_01_opt}}
 {id=WavePlayer_6, properties {bLooping=true, soundWaveAssetPtr=&/Barrier_MovingDebrisFar_01_opt}}
]
properties {maxDistance=2097152, soundClassObject=&Fort_Gadgets, volumeMultiplier=1}
