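# Sound cue for the Barrier start-moving sound, classed as a gadget sound in the Figment Season 10 Athena LTM audio.
# A quality-level node maps its first two slots to the standard wave and its last two to an "opt" variant.
# The result passes through a large barrier-wall attenuation at the root, with slightly reduced volume and a very large max distance.
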SoundCue /Figment_S10/Athena/Sounds/LTMs/Barrier/Barrier_StartMoving_Cue
root=Attenuation_0
nodes [
 {id=Attenuation_0, children [QualityLevel_0], properties {attenuationSettings=&/Barrier_Wall_Large_Att}}
 {id=QualityLevel_0, children [WavePlayer_0,WavePlayer_0,WavePlayer_1,WavePlayer_1]}
 {id=WavePlayer_0, properties {soundWaveAssetPtr=&/Barrier_StartMoving_01}}
 {id=WavePlayer_1, properties {soundWaveAssetPtr=&/Barrier_StartMoving_01_opt}}
]
properties {maxDistance=2097152, soundClassObject=&Fort_Gadgets, volumeMultiplier=0.9}
