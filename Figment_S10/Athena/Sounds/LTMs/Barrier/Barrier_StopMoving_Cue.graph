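# Sound cue for a barrier coming to a stop in the Figment_S10 LTM audio, assigned to the Fort_Gadgets sound class.
# A quality switch gives the first two levels a mix of the stop sound and a distant cube impact borrowed from event audio.
# The last two levels play a single optimized stop wave instead.
# Both variants pass through a large barrier wall attenuation at the root.
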SoundCue /Figment_S10/Athena/Sounds/LTMs/Barrier/Barrier_StopMoving_Cue
root=Attenuation_0
nodes [
 {id=Attenuation_0, children [QualityLevel_0], properties {attenuationSettings=&/Barrier_Wall_Large_Att}}
 {id=Mixer_0, children [WavePlayer_2,WavePlayer_0], properties {inputVolume [1,1]}}
 {id=QualityLevel_0, children [Mixer_0,Mixer_0,WavePlayer_1,WavePlayer_1]}
 {id=WavePlayer_0, properties {soundWaveAssetPtr=&/Barrier_StopMoving_01}}
 {id=WavePlayer_1, properties {soundWaveAssetPtr=&/Barrier_StopMoving_01_opt}}
 {id=WavePlayer_2, properties {soundWaveAssetPtr=/Figment_S10/Athena/Sounds/Events/Cube/Cube_Impact2PartDistant_01}}
]
properties {maxDistance=2097152, soundClassObject=&Fort_Gadgets, volumeMultiplier=1}
